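# Animation Blueprint for the SwissKale pickaxe that slides its dyn_knife bone out while the player holds a melee weapon.
# Each update reads the owner's FortPlayerAnimInstance weapon animation type, setting Knife_out after a short delay for Melee and clearing it for other types.
# Knife_out drives bool-alpha Modify Bone nodes, which ease an additive bone-space offset over the cached input pose.
# BodyMesh and CharacterHoldingPickaxe are set but never drive the pose.
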AnimBlueprint /BRCosmetics/Weapons/FORT_Melee/Pickaxe_SwissKale/Meshes/SwissKale_Axe_AnimBP
target_skeleton=&/SwissKale_Axe_Skeleton
graphs [
 AnimGraph {nodes [
   n1 = Get Knife_out() -> (Knife_out:bool)
   n2 = "Transform (Modify) Bone - Bone: dyn_knife"!(
    ComponentPose=Local To Component(LocalPose=Use cached pose 'Null_pose'())
    bAlphaBoolEnabled:bool=@n1
    Translation="12.5,0,0"
    Rotation="0,0,0"
    Scale="1,1,1"
   ) -> (Pose) {
    alphaInputType=Bool
    translationMode=BMM_Additive
    translationSpace=BCS_BoneSpace
    alphaBoolBlend {blendInTime=1, blendOption=QuinticInOut, blendOutTime=1}
    alphaScaleBiasClamp {bInterpResult=true}
    boneToModify {boneName=dyn_knife}
    translation {x=12.5}
   }
   n3 = "Transform (Modify) Bone - Bone: dyn_knife"(
    ComponentPose=Local To Component(LocalPose=Use cached pose 'Null_pose'())
    bAlphaBoolEnabled=True
    Translation="12.5,0,0"
    Rotation="0,0,0"
    Scale="1,1,1"
   ) {
    alphaInputType=Bool
    translationMode=BMM_Additive
    translationSpace=BCS_BoneSpace
    alphaBoolBlend {blendInTime=1, blendOption=QuinticInOut, blendOutTime=1}
    alphaScaleBiasClamp {bInterpResult=true}
    boneToModify {boneName=dyn_knife}
    translation {x=12.5}
   }
   n4 = "Transform (Modify) Bone - Bone: dyn_knife"(ComponentPose=@n3, bAlphaBoolEnabled=@n1, Translation="12.5,0,0", Rotation="0,0,0", Scale="1,1,1") {
    alphaInputType=Bool
    translationMode=BMM_Additive
    translationSpace=BCS_BoneSpace
    alphaBoolBlend {blendInTime=1, blendOption=QuinticInOut, blendOutTime=1}
    alphaScaleBiasClamp {bInterpResult=true}
    boneToModify {boneName=dyn_knife}
    translation {x=12.5}
   }
   Output Pose(Result=Blend Poses by bool(BlendPose_0=Component To Local(ComponentPose=@n4), BlendPose_1=Component To Local(ComponentPose=@n2)) {blendType=Linear}) {layerGroup=None}
   AnimGraphNode_SubInput_0 = Input Pose()
   Save cached pose 'Null_pose'(Pose=@AnimGraphNode_SubInput_0) {cacheName=Null_pose}
  ]}
 EventGraph {nodes [
   Event BlueprintUpdateAnimation()
   > n1 = Sequence()
   > n2 = Cast To FortPlayerAnimInstance(Object=GetAnimInstance(self=GetOwnerSkeletalMeshComponent()))
   > n3 = Switch on EFortWeaponCoreAnimation!(execute, Selection:EFortWeaponCoreAnimation Enum=Get WeaponCoreAnim(self=@n2)) -> (
    Melee:Exec
    Pistol:Exec
    Shotgun:Exec
    PaperBlueprint:Exec
    Rifle:Exec
    MeleeOneHand:Exec
    MachinePistol:Exec
    RocketLauncher:Exec
    GrenadeLauncher:Exec
    GoingCommando:Exec
    AssaultRifle:Exec
    TacticalShotgun:Exec
    SniperRifle:Exec
    TrapPlacement:Exec
    ShoulderLauncher:Exec
    AbilityDecoTool:Exec
    Crossbow:Exec
    C4:Exec
    RemoteControl:Exec
    DualWield:Exec
    AR_BullPup:Exec
    AR_ForwardGrip:Exec
    MedPackPaddles:Exec
    SMG_P90:Exec
    AR_DrumGun:Exec
    Consumable_Small:Exec
    Consumable_Large:Exec
    Balloon:Exec
    MountedTurret:Exec
    CreativeTool:Exec
    ExplosiveBow:Exec
    AshtonIndigo:Exec
    AshtonChicago:Exec
    MeleeDualWield:Exec
    Unarmed:Exec
   )
   Set Knife_out(
    execute=[
     @n3.AR_BullPup
     @n3.AR_DrumGun
     @n3.AR_ForwardGrip
     @n3.AbilityDecoTool
     @n3.AshtonChicago
     @n3.AshtonIndigo
     @n3.AssaultRifle
     @n3.Balloon
     @n3.C4
     @n3.Consumable_Large
     @n3.Consumable_Small
     @n3.CreativeTool
     @n3.Crossbow
     @n3.DualWield
     @n3.ExplosiveBow
     @n3.GoingCommando
     @n3.GrenadeLauncher
     @n3.MachinePistol
     @n3.MedPackPaddles
     @n3.MeleeDualWield
     @n3.MeleeOneHand
     @n3.MountedTurret
     @n3.PaperBlueprint
     @n3.Pistol
     @n3.RemoteControl
     @n3.Rifle
     @n3.RocketLauncher
     @n3.SMG_P90
     @n3.Shotgun
     @n3.ShoulderLauncher
     @n3.SniperRifle
     @n3.TacticalShotgun
     @n3.TrapPlacement
     @n3.Unarmed
    ]
    Knife_out:bool=false
   ) -> (Output_Get:bool)
   Branch(execute=@n3.Melee)
   > Delay(Duration=0.09)
   > Set Knife_out(Knife_out=true)
   n4 = Get!(execute:Exec=@n1.then_1) -> (then:Exec, else:Exec, BodyMesh:Skeletal Mesh Component Budgeted Object Reference)
   > Set CharacterHoldingPickaxe(CharacterHoldingPickaxe:bool=false) -> (Output_Get:bool)
   Set CharacterHoldingPickaxe(execute=@n4.else, CharacterHoldingPickaxe=true)
   n5 = Event OnCharacterCustomizationLoaded()
   > Set BodyMesh(BodyMesh:Skeletal Mesh Component Budgeted Object Reference=Get Array Item with EFort Custom Part Type(Array=@n5.CustomMeshParts)) -> (Output_Get:Skeletal Mesh Component Budgeted Object Reference)
  ]}
]
variables [
 BodyMesh {type=SkeletalMeshComponentBudgeted}
 CharacterHoldingPickaxe {type=boolean}
 Idle {type=number}
 Knife_out {type=boolean}
 PICKAXE_0 {type=number}
]
